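# Physics asset for the Squash vehicle body cosmetic, apparently for the SK_VEH_Squash skeletal mesh, providing a simplified box collision shape.
# It holds a single body on the root bone, made of one box roughly car-sized, raised above the origin and shifted slightly along positive X.
# It has no constraints and no other bone bodies, so the whole mesh is represented by one volume.
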PhysicsAsset /VehicleCosmetics/Bodies/Squash/Meshes/SK_VEH_Squash_BoxPhysics
bodies {root [{shapeType=Box, extentX=444, extentY=216, extentZ=116, shapeName=root_box, center=<37.0367 0 95.7678>}]}
constraints []
