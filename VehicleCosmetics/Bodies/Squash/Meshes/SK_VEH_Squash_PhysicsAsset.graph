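# Physics asset for the Squash vehicle body skeletal mesh in VehicleCosmetics.
# It defines a single body on the root bone, with no collision shapes listed.
# It has no constraints, so the asset sets up no articulated or ragdoll motion between bones.
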PhysicsAsset /VehicleCosmetics/Bodies/Squash/Meshes/SK_VEH_Squash_PhysicsAsset
bodies {root []}
constraints []
